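# Sound cue for the SixShooter pistol's cocking foley, routed to the weapon foley sound class at reduced volume.
# A Random node picks one of three equally weighted cocking wave variations on each play.
# Overridden attenuation supersedes the referenced weapon reload asset, using logarithmic falloff over a very short range with a Direct2D non-spatialized radius.
# Overridden concurrency prevents new instances once the limit is reached.
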SoundCue /Game/Sounds/Fort_Weapons/pistol/SixShooter/SC_SixShooter_Cocking
root=Random_0
nodes [
 {id=Random_0, children [WavePlayer_0,WavePlayer_2,WavePlayer_3], properties {weights [1,1,1]}}
 {id=WavePlayer_0, properties {soundWaveAssetPtr=&/SixShooter_Cocking_02}}
 {id=WavePlayer_2, properties {soundWaveAssetPtr=&/SixShooter_Cocking_01}}
 {id=WavePlayer_3, properties {soundWaveAssetPtr=&/SixShooter_Cocking_03}}
]
properties {
 attenuationSettings=&Fort_Weapon_Reload_Att
 bOverrideAttenuation=true
 bOverrideConcurrency=true
 maxDistance=100
 soundClassObject=&Fort_WeaponFoley
 volumeMultiplier=0.3
 attenuationOverrides {
  distanceAlgorithm=Logarithmic
  falloffDistance=0
  nonSpatializedRadiusEnd=200
  nonSpatializedRadiusMode=Direct2D
  nonSpatializedRadiusStart=200
  stereoSpread=0
  attenuationShapeExtents {x=100}
 }
 concurrencyOverrides {resolutionRule=PreventNew}
}
