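# Sound cue for the SixShooter pistol drop, built from Pistol_CherrySmoke drop samples.
# A quality-level switch gives its first two levels a random pick of three equally weighted waves and its last two only the first wave.
# The result passes through a volume-boosting mixer into a root modulator.
# It uses weapon-drop attenuation, the Fort_Loot_Collect sound class and the Weapon_Drop_StopOldest concurrency set.
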SoundCue /Game/Sounds/Fort_Weapons/pistol/SixShooter/SC_SixShooter_Drop
root=Modulator_0
nodes [
 {id=Mixer_0, children [QualityLevel_0], properties {inputVolume [2.2]}}
 {id=Modulator_0, children [Mixer_0]}
 {id=QualityLevel_0, children [Random_0,Random_0,WavePlayer_0,WavePlayer_0]}
 {id=Random_0, children [WavePlayer_0,WavePlayer_1,WavePlayer_2], properties {weights [1,1,1]}}
 {id=WavePlayer_0, properties {soundWaveAssetPtr=&/Pistol_CherrySmoke_Drop_01}}
 {id=WavePlayer_1, properties {soundWaveAssetPtr=&/Pistol_CherrySmoke_Drop_02}}
 {id=WavePlayer_2, properties {soundWaveAssetPtr=&/Pistol_CherrySmoke_Drop_03}}
]
properties {
 attenuationSettings=&Fort_Weapon_Drop_Att
 maxDistance=3000
 soundClassObject=&Fort_Loot_Collect
 volumeMultiplier=1
 concurrencyOverrides {maxCount=1, resolutionRule=PreventNew}
 concurrencySet [&Weapon_Drop_StopOldest]
}
