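# Sound cue for a dropped SixShooter pistol landing, routed to the loot collect sound class.
# A single SixShooter_Land wave plays through a Modulator at default settings for slight random pitch and volume variation, with overall volume reduced.
# Shared weapon-drop attenuation handles spatial falloff.
# It joins the Weapon_Drop_StopOldest concurrency set; the stored single-voice PreventNew override values show no enabled override flag.
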SoundCue /Game/Sounds/Fort_Weapons/pistol/SixShooter/SC_SixShooter_Drop_Land
root=Modulator_0
nodes [
 {id=Modulator_0, children [WavePlayer_0]}
 {id=WavePlayer_0, properties {soundWaveAssetPtr=&/SixShooter_Land}}
]
properties {
 attenuationSettings=&Fort_Weapon_Drop_Att
 maxDistance=3000
 soundClassObject=&Fort_Loot_Collect
 volumeMultiplier=0.7
 concurrencyOverrides {maxCount=1, resolutionRule=PreventNew}
 concurrencySet [&Weapon_Drop_StopOldest]
}
