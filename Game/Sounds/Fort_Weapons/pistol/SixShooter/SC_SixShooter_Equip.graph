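# Sound cue for the equip foley of the SixShooter pistol.
# A single wave player feeds the root directly, with no randomization or mixing, at reduced volume in the weapon foley sound class.
# Spatial falloff comes from the shared weapon reload attenuation asset, as the cue's attenuation override values are not switched on.
# Its own concurrency override blocks new instances once the limit is reached.
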SoundCue /Game/Sounds/Fort_Weapons/pistol/SixShooter/SC_SixShooter_Equip
root=WavePlayer_0
nodes [{id=WavePlayer_0, properties {soundWaveAssetPtr=&/SixShooter_Equip_01}}]
properties {
 attenuationSettings=&Fort_Weapon_Reload_Att
 bOverrideConcurrency=true
 maxDistance=2250
 soundClassObject=&Fort_WeaponFoley
 volumeMultiplier=0.501
 attenuationOverrides {falloffDistance=2000, nonSpatializedRadiusStart=200, attenuationShapeExtents {x=500}}
 concurrencyOverrides {resolutionRule=PreventNew}
}
