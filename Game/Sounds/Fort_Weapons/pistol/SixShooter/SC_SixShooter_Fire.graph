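# Sound cue for the SixShooter pistol's gunshot, built as three distance layers mixed together with the close layer loudest.
# Close, medium and distant layers each pick randomly from four waves, pass through a volume modulator, then use their own attenuation asset.
# The distant layer is also pitched down, and sound class overrides route close to a 3P fire class and the other two to a distant class.
# Local and global gunfire concurrency sets limit playback.
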SoundCue /Game/Sounds/Fort_Weapons/pistol/SixShooter/SC_SixShooter_Fire
root=Mixer_2
nodes [
 {
  id=Attenuation_119
  children [Modulator_2]
  properties {attenuationSettings=&Fort_Weapon_Projectile_Medium_Mono_Athena_Att}
 }
 {
  id=Attenuation_2
  children [Modulator_5]
  properties {attenuationSettings=&Fort_Weapon_Projectile_Distant_Mono_Athena_Att}
 }
 {id=Attenuation_3, children [Modulator_6], properties {attenuationSettings=&Fort_Weapon_Projectile_Close_Athena_Att}}
 {id=Mixer_2, children [SoundClass_3, SoundClass_76, SoundClass_2], properties {inputVolume [1.3,0.45,0.3]}}
 {id=Modulator_2, children [Random_0], properties {volumeMax=1.6, volumeMin=1.4}}
 {id=Modulator_5, children [Random_1], properties {pitchMax=0.9, pitchMin=0.8, volumeMax=1.4, volumeMin=1.2}}
 {id=Modulator_6, children [Random_3], properties {volumeMax=1.1, volumeMin=1}}
 {
  id=Random_0
  children [WavePlayer_4,WavePlayer_5,WavePlayer_6,WavePlayer_7]
  properties {bShouldExcludeFromBranchCulling=true, weights [1,1,1,1]}
 }
 {
  id=Random_1
  children [WavePlayer_8, WavePlayer_9, WavePlayer_10, WavePlayer_11]
  properties {bShouldExcludeFromBranchCulling=true, weights [1,1,1,1]}
 }
 {
  id=Random_3
  children [WavePlayer_0,WavePlayer_1,WavePlayer_2,WavePlayer_3]
  properties {bShouldExcludeFromBranchCulling=true, weights [1,1,1,1]}
 }
 {
  id=SoundClass_2
  children [Attenuation_2]
  properties {soundClassOverride=&Fort_WeaponRanged_Fire_Distant}
 }
 {
  id=SoundClass_3
  children [Attenuation_3]
  properties {soundClassOverride=&Fort_WeaponRanged_Fire_3P_PSM}
 }
 {
  id=SoundClass_76
  children [Attenuation_119]
  properties {soundClassOverride=&Fort_WeaponRanged_Fire_Distant}
 }
 {id=WavePlayer_0, properties {soundWaveAssetPtr=&/SixShooter_Fire_Close_01}}
 {id=WavePlayer_1, properties {soundWaveAssetPtr=&/SixShooter_Fire_Close_02}}
 {id=WavePlayer_10, properties {soundWaveAssetPtr=&/SixShooter_Fire_DistantNew_03}}
 {id=WavePlayer_11, properties {soundWaveAssetPtr=&/SixShooter_Fire_DistantNew_04}}
 {id=WavePlayer_2, properties {soundWaveAssetPtr=&/SixShooter_Fire_Close_03}}
 {id=WavePlayer_3, properties {soundWaveAssetPtr=&/SixShooter_Fire_Close_04}}
 {id=WavePlayer_4, properties {soundWaveAssetPtr=&/SixShooter_Fire_MediumNew_01}}
 {id=WavePlayer_5, properties {soundWaveAssetPtr=&/SixShooter_Fire_MediumNew_02}}
 {id=WavePlayer_6, properties {soundWaveAssetPtr=&/SixShooter_Fire_MediumNew_03}}
 {id=WavePlayer_7, properties {soundWaveAssetPtr=&/SixShooter_Fire_MediumNew_04}}
 {id=WavePlayer_8, properties {soundWaveAssetPtr=&/SixShooter_Fire_DistantNew_01}}
 {id=WavePlayer_9, properties {soundWaveAssetPtr=&/SixShooter_Fire_DistantNew_02}}
]
properties {
 maxDistance=27500
 attenuationOverrides {falloffDistance=4100, attenuationShapeExtents {x=900}}
 concurrencySet [&Weapon_Gunfire_Local_Conc, &Weapon_Gunfire_Global_Conc]
}
