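# Sound cue for the first-person gunfire of the SixShooter pistol.
# A quality-level switch sends the first level to a random pick of four close-fire waves, and the other three levels share a three-wave pick.
# Each pick runs through a volume-boosting modulator with slight random variation, then a mixer whose second input is empty.
# First-person weapon attenuation, sound class and gunfire concurrency govern playback.
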SoundCue /Game/Sounds/Fort_Weapons/pistol/SixShooter/SC_SixShooter_Fire_1P
root=QualityLevel_0
nodes [
 {id=Mixer_1, children [Modulator_1,null], properties {inputVolume [1,0.8]}}
 {id=Mixer_2, children [Modulator_6,null], properties {inputVolume [1,0.8]}}
 {id=Modulator_1, children [Random_6], properties {volumeMax=1.4, volumeMin=1.3}}
 {id=Modulator_6, children [Random_3], properties {volumeMax=1.4, volumeMin=1.3}}
 {id=QualityLevel_0, children [Mixer_2,Mixer_1,Mixer_1,Mixer_1]}
 {id=Random_3, children [WavePlayer_0,WavePlayer_1,WavePlayer_2,WavePlayer_3], properties {weights [1,1,1,1]}}
 {id=Random_6, children [WavePlayer_27, WavePlayer_28, WavePlayer_30], properties {weights [1,1,1]}}
 {id=WavePlayer_0, properties {soundWaveAssetPtr=&/SixShooter_Fire_Close_01}}
 {id=WavePlayer_1, properties {soundWaveAssetPtr=&/SixShooter_Fire_Close_02}}
 {id=WavePlayer_2, properties {soundWaveAssetPtr=&/SixShooter_Fire_Close_03}}
 {id=WavePlayer_27, properties {soundWaveAssetPtr=&SixShooter_Fire_Close_01}}
 {id=WavePlayer_28, properties {soundWaveAssetPtr=&SixShooter_Fire_Close_02}}
 {id=WavePlayer_3, properties {soundWaveAssetPtr=&/SixShooter_Fire_Close_04}}
 {id=WavePlayer_30, properties {soundWaveAssetPtr=&SixShooter_Fire_Close_04}}
]
properties {
 attenuationSettings=&Fort_Weapon_Projectile_FirstPerson_Athena_Att
 maxDistance=2097152
 soundClassObject=&Fort_WeaponRanged_Fire_1P_PSM
 attenuationOverrides {falloffDistance=4100, attenuationShapeExtents {x=900}}
 concurrencySet [&Weapon_Gunfire_FirstPerson_Conc]
}
