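# Sound cue for picking up the SixShooter pistol, classed as loot collection audio with pickup attenuation.
# A Modulator feeds a quality-level switch whose first two slots use an equal-weight random pick among three pickup waves.
# The last two slots bypass the random node and play only the first wave.
# Playback is quietened by a reduced volume multiplier and assigned to a shared pickup concurrency set.
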SoundCue /Game/Sounds/Fort_Weapons/pistol/SixShooter/SC_SixShooter_Pickup
root=Modulator_0
nodes [
 {id=Modulator_0, children [QualityLevel_0]}
 {id=QualityLevel_0, children [Random_0,Random_0,WavePlayer_1,WavePlayer_1]}
 {id=Random_0, children [WavePlayer_1,WavePlayer_2,WavePlayer_3], properties {weights [1,1,1]}}
 {id=WavePlayer_1, properties {soundWaveAssetPtr=&/SixShooter_Pickup_01}}
 {id=WavePlayer_2, properties {soundWaveAssetPtr=&/SixShooter_Pickup_02}}
 {id=WavePlayer_3, properties {soundWaveAssetPtr=&/SixShooter_Pickup_03}}
]
properties {
 attenuationSettings=&Fort_Weapon_Pickup_Att
 maxDistance=3500
 soundClassObject=&Fort_Loot_Collect
 volumeMultiplier=0.3
 concurrencyOverrides {maxCount=1, resolutionRule=PreventNew}
 concurrencySet [&PickupMaximum_Oldest]
}
