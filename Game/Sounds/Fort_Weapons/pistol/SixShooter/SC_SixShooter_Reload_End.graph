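# SoundCue for the end of the SixShooter pistol's reload sound.
# A single wave player feeds the SixShooter_ReloadEndB_02 wave straight to the root, with no randomization or modulation.
# It uses the weapon reload attenuation and the weapon foley sound class, and overrides concurrency with a rule that prevents new instances once the limit is reached.
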SoundCue /Game/Sounds/Fort_Weapons/pistol/SixShooter/SC_SixShooter_Reload_End
root=WavePlayer_7
nodes [{id=WavePlayer_7, properties {soundWaveAssetPtr=&/SixShooter_ReloadEndB_02}}]
properties {
 attenuationSettings=&Fort_Weapon_Reload_Att
 bOverrideConcurrency=true
 maxDistance=2250
 soundClassObject=&Fort_WeaponFoley
 volumeMultiplier=1.002
 concurrencyOverrides {resolutionRule=PreventNew}
}
