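# Sound cue for the start of the Six Shooter pistol reload, played as weapon foley with the shared reload attenuation.
# A mixer layers a primary reload wave, slightly varied in pitch and volume, over a quieter secondary layer.
# The secondary layer starts after a fixed half-second delay and randomly picks one of two ReloadStartB variants.
# A concurrency override prevents new instances once the limit is reached.
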SoundCue /Game/Sounds/Fort_Weapons/pistol/SixShooter/SC_SixShooter_Reload_Start
root=Mixer_1
nodes [
 {id=Delay_1, children [Random_0], properties {delayMax=0.5, delayMin=0.5}}
 {id=Mixer_1, children [Modulator_0,Delay_1], properties {inputVolume [1.4,0.4]}}
 {id=Modulator_0, children [WavePlayer_5], properties {pitchMax=1.1, pitchMin=1, volumeMax=1.2, volumeMin=1.1}}
 {id=Random_0, children [WavePlayer_3,WavePlayer_4], properties {weights [1,1]}}
 {id=WavePlayer_3, properties {soundWaveAssetPtr=&/SixShooter_ReloadStartB_01}}
 {id=WavePlayer_4, properties {soundWaveAssetPtr=&/SixShooter_ReloadStartB_02}}
 {id=WavePlayer_5, properties {soundWaveAssetPtr=&/SixShooter_ReloadStart_02}}
]
properties {
 attenuationSettings=&Fort_Weapon_Reload_Att
 bOverrideConcurrency=true
 maxDistance=2250
 soundClassObject=&Fort_WeaponFoley
 volumeMultiplier=0.752
 concurrencyOverrides {resolutionRule=PreventNew}
}
